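# Third-person sound cue for the male Black Monday emote, mixing a foley wave with a looping ambience.
# The ambience runs through an Enveloper whose volume curve starts silent, swells to a peak and settles over ten seconds, then a Modulator that lowers it with slight random variation.
# It uses the third-person emote music class, emote attenuation and concurrency, plays when silent, and sends to a music audio-analysis submix.
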SoundCue /BRCosmetics/Sounds/Emotes/BlackMondayMale/SC_Athena_Emote_BlackMondayMale_3P_Cue
root=Mixer_0
nodes [
 {
  id=Enveloper_1
  children [WavePlayer_2]
  properties {
   pitchCurve {editorCurveData {keys [{time=0}]}}
   volumeCurve {editorCurveData {keys [
      <at 0 0>
      {arriveTangent=-1.3759, interpMode=RCIM_Cubic, leaveTangent=-1.3759, tangentMode=RCTM_User, time=1.5, value=0.5}
      {arriveTangent=0.0447, interpMode=RCIM_Cubic, leaveTangent=0.0447, tangentMode=RCTM_User, time=2.5, value=2}
      {arriveTangent=-0.0662, leaveTangent=-0.0662, time=10, value=1}
     ]}}
  }
 }
 {id=Mixer_0, children [WavePlayer_0,Modulator_0], properties {inputVolume [1,1]}}
 {id=Modulator_0, children [Enveloper_1], properties {volumeMax=0.45, volumeMin=0.4}}
 {id=WavePlayer_0, properties {soundWaveAssetPtr=&/Athena_Emote_BlackMondayMale_Foley}}
 {
  id=WavePlayer_2
  properties {bLooping=true, soundWaveAssetPtr=&/Athena_Emote_BlackMondayMale_Amb_Loop}
 }
]
properties {
 attenuationSettings=&Fort_Emotes_Athena_Att
 maxDistance=1600
 soundClassObject=/Game/Sounds/FortSoundClassesAndModes/Fort_Music_Emote_3P
 virtualizationMode=PlayWhenSilent
 volumeMultiplier=1
 concurrencyOverrides {resolutionRule=PreventNew}
 concurrencySet [/Game/Sounds/Concurrency_Rules/Athena_EmoteMusic_Conc]
 soundSubmixSends [{minSendLevel=0, soundSubmix=&Fortnite_AudioAnalysis_Music}]
}
